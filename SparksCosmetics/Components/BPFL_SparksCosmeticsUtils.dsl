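; Blueprint function library of Sparks cosmetics utilities that supplies backpack attachment transforms for instruments.
; One function each for bass, drum, guitar, keytar and mic sets and returns a hard-coded backpack offset of location, rotation and scale.
; Bass, guitar and keytar share identical values, while drum and mic have their own.
; In-graph comments explain what each axis controls, such as body clipping, tilt and facing.
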(fn BackpackBassOffset
  (bind _output_get (Variables|SetBackpackOffset_Location (struct :Location "-2.000000,-7.500000,-6.000000" :Rotation "0.000000,-90.000000,20.000000" :Scale "0.850000,0.850000,0.850000"))) ; BASS TRANSFORMS: \nTX = horizontal\nTY = vertical\nTZ = depth (body-clipping)\nRX = tilt (neck/fret)\nRZ = facing (strings face camera)
  (return _output_get))

(fn BackpackDrumOffset
  (bind _output_get (Variables|SetBackpackOffset_Location (struct :Location "-5.000000,-7.500000,10.000000" :Rotation "0.000000,-90.000000,-210.000000" :Scale "0.850000,0.850000,0.850000"))) ; DRUM TRANSFORMS: \nTX = horizontal\nTY = vertical\nTZ = depth (body-clipping)\nRX = tilt (neck/fret)\nRZ = facing (strings face camera)
  (return _output_get))

(fn BackpackGuitarOffset
  (bind _output_get (Variables|SetBackpackOffset_Location (struct :Location "-2.000000,-7.500000,-6.000000" :Rotation "0.000000,-90.000000,20.000000" :Scale "0.850000,0.850000,0.850000"))) ; GUITAR TRANSFORMS: \nTX = horizontal\nTY = vertical\nTZ = depth (body-clipping)\nRX = tilt (neck/fret)\nRZ = facing (strings face camera)
  (return _output_get))

(fn BackpackKeytarOffset
  (bind _output_get (Variables|SetBackpackOffset_Location (struct :Location "-2.000000,-7.500000,-6.000000" :Rotation "0.000000,-90.000000,20.000000" :Scale "0.850000,0.850000,0.850000"))) ; KEYTAR TRANSFORMS: \nTX = horizontal\nTY = vertical\nTZ = depth (body-clipping)\nRX = tilt (neck/fret)\nRZ = facing (keys face camera)
  (return _output_get))

(fn BackpackMicOffset
  (bind _output_get (Variables|SetBackpackOffset_Location (struct :Location "12.000000,-2.000000,-10.000000" :Rotation "15.000000,0.000000,17.000000" :Scale "1.000000,1.000000,1.000000"))) ; MIC TRANSFORMS: \nTX = horizontal\nTY = vertical\nTZ = depth (body-clipping)\nRX = tilt (micstand)\nRZ = facing (microphone faces away from camera)
  (return _output_get))
