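; Actor component on a Sparks instrument weapon that records the player's instrument choice when that weapon is equipped.
; From BeginPlay to EndPlay it listens to the owning player pawn's weapon equip and unequip delegates.
; A WasEquipped flag guards against double transitions, with optional debug prints.
; On a new equip, only the authority passes InstrumentType to the pawn PlayerState's SparksPlayerStateComponent.
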(event EventBeginPlay
  ;[0.33 0.66 3] Begin listening for equipped
  (CallFunction|InitializeEquipListeners))

(event EventEndPlay
  (params (EndPlayReason :type EEndPlayReason%20Enum :default "Destroyed"))
  ;[0.33 0.66 3] Begin listening for equipped
  (CallFunction|CleanupEquipListeners)
  ;[0.33 0.66 3] Make sure our listeners and such get cleaned up
  (CallFunction|OnUnEquipped))

(event Custom|OnEquipped
  (params (NewWeapon :type Fort%20Weapon%20Object%20Reference) (PrevWeapon :type Fort%20Weapon%20Object%20Reference))
  (bind _returnvalue (Components|GetOwner))
  ;[3 1 0.66] Make sure that this is the weapon that was just equipped
  (if (== NewWeapon _returnvalue)
    ;[0.33 3 0.66] Make sure we have an actual state change
    (if (Variables|Default|GetWasEquipped)
      (Development|PrintString "Don't Double Equip" (Variables|Default|GetSHOW%20DEBUG%20TEXT) (Variables|Default|GetSHOW%20DEBUG%20TEXT) "(R=0.000000,G=0.640625,B=0.000000,A=1.000000)" 5.0) ; No change in state - do nothing
      (else
        (Variables|Default|SetWasEquipped true)
        (CallFunction|UpdateInstrumentSelection)))
    (else
      ;[3 1 0.66] If it was our old instrument, clean up
      (if (== _returnvalue PrevWeapon)
        (if (Variables|Default|GetWasEquipped)
          (Variables|Default|SetWasEquipped false)
          (else
            (Development|PrintString "Don't Double UnEquip" (Variables|Default|GetSHOW%20DEBUG%20TEXT) (Variables|Default|GetSHOW%20DEBUG%20TEXT) "(R=0.000000,G=0.640625,B=0.000000,A=1.000000)" 5.0) ; No change in state - do nothing
          ))))))

(event Custom|OnUnEquipped
  ;[0.33 3 0.66] Make sure we have an actual state change
  (if (Variables|Default|GetWasEquipped)
    (Variables|Default|SetWasEquipped false)
    (else
      (Development|PrintString "Don't Double UnEquip" (Variables|Default|GetSHOW%20DEBUG%20TEXT) (Variables|Default|GetSHOW%20DEBUG%20TEXT) "(R=0.000000,G=0.640625,B=0.000000,A=1.000000)" 5.0) ; No change in state - do nothing
    )))

(fn CleanupEquipListeners
  (bind _returnvalue (Components|GetOwner))
  (bind _returnvalue_2 (Actor|GetOwner _returnvalue))
  (Utilities|IsValid _returnvalue
    (:Is%20Valid
      (Utilities|IsValid _returnvalue_2
        (:Is%20Valid
          ;[0.33 3 0.66] Stop listening for equip status changes
          (bind _asfort_player_pawn (Utilities|Casting|CastToFortPlayerPawn _returnvalue_2)
            (:then
              (FortPawn|UnbindEventfromOnWeaponEquippedDelegate _asfort_player_pawn (EventDispatchers|CreateEvent :!event OnEquipped))
              (FortPawn|UnbindEventfromOnWeaponUnEquippedDelegate _asfort_player_pawn (EventDispatchers|CreateEvent :!event OnUnEquipped))
              ;[3 0.66 3] Finished cleaning up equip listeners
              (return))
            (:CastFailed
              ;[3 0.66 3] Nothing to clean up
              (return))))))))

(fn GetInstrumentType
  (return (Variables|Default|GetInstrumentType)))

(fn InitializeEquipListeners
  ;[0.33 3 0.66] Listen for equip status changes
  (bind _asfort_player_pawn (Utilities|Casting|CastToFortPlayerPawn (Actor|GetOwner (Components|GetOwner)))
    (:then
      (FortPawn|BindEventtoOnWeaponEquippedDelegate _asfort_player_pawn (EventDispatchers|CreateEvent :!event OnEquipped))
      (FortPawn|BindEventtoOnWeaponUnEquippedDelegate _asfort_player_pawn (EventDispatchers|CreateEvent :!event OnUnEquipped))
      ;[3 0.66 3] Finished setting up equip listeners
      (return))
    (:CastFailed
      ;[3 0.66 3] Nothing to listen to
      (return))))

(fn UpdateInstrumentSelection
  (Networking|SwitchHasAuthority
    (:Authority
      (bind _asfort_pawn (Utilities|Casting|CastToFortPawn (Actor|GetOwner (Components|GetOwner))))
      (Utilities|IsValid (Variables|/Script/Engine.Pawn|GetPlayerState _asfort_pawn)
        (:Is%20Valid
          (Utilities|IsValid (Actor|GetComponentbyClass (Variables|/Script/Engine.Pawn|GetPlayerState _asfort_pawn) "/Script/SparksCosmeticsRuntime.SparksPlayerStateComponent")
            (:Is%20Valid
              (Class|SparksPlayerStateComponent|SetSparksInstrumentChoice (Actor|GetComponentbyClass (Variables|/Script/Engine.Pawn|GetPlayerState _asfort_pawn) "/Script/SparksCosmeticsRuntime.SparksPlayerStateComponent") (Variables|Default|GetInstrumentType)))))))))
